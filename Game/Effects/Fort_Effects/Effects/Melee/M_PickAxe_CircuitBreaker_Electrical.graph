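# Unlit, two-sided translucent particle-mesh material for the electrical effect on the Circuit Breaker pickaxe.
# Particle color feeds emissive adjustments, while a panned beam texture, tiled and eroded via dynamic parameters, builds opacity with particle alpha.
# A V-gradient masks both edges, and depth fade applies off mobile.
# World-space animated vector noise forms an offset with a constant mobile fallback; the time-driven lerp blends identical inputs.
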Material /Game/Effects/Fort_Effects/Effects/Melee/M_PickAxe_CircuitBreaker_Electrical
graph {
 properties {
  bEnableResponsiveAA=true
  bUsedWithMeshParticles=true
  bUsedWithNiagaraMeshParticles=true
  blendMode=BLEND_Translucent
  pixelDepthOffsetMode=PDOM_Legacy
  shadingModel=MSM_Unlit
  twoSided=true
  usedShadingModels=Unlit
 }
 nodes [
  MaterialFunctionCall(Emissive=ParticleColor().RGB) {materialFunction=&MF_FX_EmissiveAdjustments}
  n1 = DynamicParameter() {defaultValue=<rgba 1 0 1 1>, paramNames [Panner Speed, Erode, U Tile, LERP]}
  Multiply(A=Time(), B=@n1.Panner Speed)
  n2 = TextureCoordinate()
  n3 = ComponentMask(None=@n2) {g=true}
  n4 = OneMinus(None=ComponentMask(None=@n2) {g=true})
  n5 = ScalarParameter() {defaultValue=1, parameterName=V Tile}
  n6 = DynamicParameter() {defaultValue=<rgba 1 0 1 1>, paramNames [Panner Speed, Erode, U Tile, LERP]}
  n7 = Multiply(A=TextureCoordinate(), B=AppendVector(A=ScalarParameter() {defaultValue=0.7, parameterName=U Tile}, B=Multiply(A=@n5, B=@n6.U Tile)))
  n8 = TextureSample(UVs=Panner(Coordinate=Add(A=@n7, B=@n6.U Tile), Speed=@n1.Panner Speed) {bFractionalPart=true, speedY=1}) {samplerType=SAMPLERTYPE_LinearColor, texture=/Game/Effects/Fort_Effects/Textures/Tile/T_Raygun_Beam_02}
  n9 = Clamp(None=Multiply(A=ParticleRelativeTime(), B=ScalarParameter() {defaultValue=1.35, parameterName=PRT Scale}))
  n10 = DynamicParameter() {defaultValue=<rgba 1 0 1 1>, paramNames [Panner Speed, Erode, U Tile, LERP]}
  n11 = LinearInterpolate(
   A=@n8.G
   B=@n8.G
   Alpha=MaterialFunctionCall!(In=StaticSwitchParameter(True=@n9, False=@n10.LERP) {defaultValue=true, parameterName=Use Particle Time LERP?}) {materialFunction=/Engine/Functions/Engine_MaterialFunctions02/Utility/Round}.Result
  ) {constAlpha=1}
  n12 = Add(A=@n11, B=DynamicParameter() {defaultValue=<rgba 1 0 1 1>, paramNames [Panner Speed, Erode, U Tile, LERP]}.Erode)
  n13 = Multiply(A=ParticleColor().A, B=Clamp(None=Multiply(A=@n11, B=@n12)))
  n14 = ScalarParameter() {defaultValue=5, parameterName=FadeDistance}
  Multiply(
   A=Multiply(A=Constant() {r=2}, B=Multiply(A=@n3, B=@n4))
   B=ShadingPathSwitch(Default=DepthFade(Opacity=@n13, FadeDistance=@n14), Mobile=@n13) {desc=Migrated from FeatureLevelSwitch}
  )
  n15 = WorldPosition()
  n16 = ActorPositionWS()
  n17 = ScalarParameter() {defaultValue=3, parameterName=Curl_Speed}
  n18 = Time()
  n19 = Divide(A=Subtract(A=Subtract(A=@n15.XYZ, B=@n16), B=Multiply(A=@n17, B=@n18)), B=ScalarParameter() {defaultValue=2, parameterName=Curl_Scale})
  ShadingPathSwitch(
   Default=Multiply(A=ScalarParameter() {defaultValue=1, parameterName=Curl_Intensity}, B=VectorNoise(World Position=@n19) {noiseFunction=VNF_VectorALU})
   Mobile=Constant() {r=1}
  ) {desc=Migrated from FeatureLevelSwitch}
 ]
}
